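; Quest gameplay ability for FearlessFlight that credits the player for swinging onto a grind rail.
; On activation it waits for the Swinging movement tag to be removed, then checks quest context tags for the WestSausage item in inventory.
; After a short delay, giving grinding time to register, it checks for the Grinding tag.
; If present, it sends the Season 24 WebSwingGrindRail stat event and ends; failed checks do not end it.
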(event Ability|EventActivateAbility
  (|CallParentFunction)
  (Ability|Tasks|WaitGameplayTagRemove "(TagName=\"MovementMode.Swinging\")"
    (:Removed
      (if (GameplayTags|HasTag (Quests|GetSourceandContextTags (Variables|GAVars|Data|GetOwningPlayerQuestManager)) "(TagName=\"Granted.Athena.WestSausage.InInventory\")" true)
        (Utilities|FlowControl|Delay 0.2)
        (if (GameplayTags|HasTag (Quests|GetSourceandContextTags (Variables|GAVars|Data|GetOwningPlayerQuestManager)) "(TagName=\"MovementMode.Grinding\")" true)
          (Quest|SendComplexCustomStatEvent (Variables|GAVars|Data|GetOwningPlayerQuestManager) 0 "(GameplayTags=((TagName=\"Athena.Quests.S24.WebSwingGrindRail\")))" "(GameplayTags=)" 1)
          (Ability|EndAbility))))))
